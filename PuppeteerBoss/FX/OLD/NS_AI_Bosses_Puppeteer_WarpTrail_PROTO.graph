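# Prototype Niagara ribbon trail for the Puppeteer boss warp, stretched from the system's position to a target position.
# A single burst spawns a few particles spaced along that line by normalized index, lifted upward with slight jitter and linked in that order.
# Ribbons widen over their short life in a near-black tint, with alpha fading over lifetime and toward both ends.
# User.IsRoseDepthTeleport flips the sign of the U pan speed variable.
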NiagaraSystem /PuppeteerBoss/FX/OLD/NS_AI_Bosses_Puppeteer_WarpTrail_PROTO
dependencies {
 usedDataInterfaces [&NiagaraDataInterfaceCurve, &NiagaraDataInterfaceColorCurve, &NiagaraDataInterfaceVectorCurve]
 usedDynamicInputs [
  &VectorFromCurve
  &TransformVector
  &NormalizeVector
  &Subtract_Position
  /Niagara/DynamicInputs/Helpers/SimulationPosition
 ]
 usedModules [
  &self:SystemSpawnScript.SystemScriptSource.SystemScriptGraph.NiagaraNodeAssignment_0.SetVariables_#1
  &self:SystemSpawnScript.SystemScriptSource.SystemScriptGraph.NiagaraNodeAssignment_1.SetVariables_#2
  &SystemState
 ]
}
emitters {Trail002 {
  emitterSpawnScript [/SetVariables(Emitter.PanSpeedU:NiagaraFloat=&MakeCustomFloatFromBool(Bool=$User.IsRoseDepthTeleport, True Float=2, False Float=-2)) {isSetParameters=true}]
  emitterUpdateScript [
   &EmitterState()
   &SpawnBurst_Instantaneous(Spawn Count=5, Use Spawn Probability=false, UseLoopCountLimit=false)
  ]
  particleSpawnScript [
   /SetVariables(Particles.RibbonLinkOrder:NiagaraFloat=&ReturnNormalizedExecIndex()) {isSetParameters=true}
   &InitializeParticle(
    Color Minimum=<rgba 0.7 0.7 0.7 1>
    Color Mode=::NewEnumerator2
    Lifetime Max=1.2
    Lifetime Min=0.8
    Lifetime Mode=::NewEnumerator1
    Position=&LerpPosition(A=$Engine.Emitter.SimulationPosition, B=$System.TargetPosition, Alpha=&ReturnNormalizedExecIndex())
    Position Mode=::NewEnumerator1
    Ribbon Width=&RandomRangeFloat(Minimum=250, Maximum=300)
    Ribbon Width Mode=::NewEnumerator1
    Sprite Size Mode=::NewEnumerator3
    Uniform Sprite Size=5
    UsePositionOffset=false
    Write Material Random=true
   )
   /SetVariables(Particles.Position=&AddVectorToPosition(Position=$Particles.Position, Vector=&RandomRangeVector(Minimum=<-10 -10 80>, Maximum=<10 10 90>))) {isSetParameters=true}
   &FN_RibbonMaterial(Color Pan U Offset:NiagaraFloat=&RandomRangeFloat(Minimum=0, Maximum=1))
  ]
  particleUpdateScript [
   &ParticleState()
   &ScaleRibbonWidth(Ribbon Width Scale=&FloatFromCurve(FloatCurve=DataInterface() {propertyValues {Curve {keys [
         {arriveTangent=4.1785, interpMode=RCIM_Cubic, leaveTangent=4.1785, tangentMode=RCTM_User, time=0, value=0}
         {arriveTangent=1.1179, interpMode=RCIM_Cubic, leaveTangent=1.1179, tangentMode=RCTM_User, time=0.2042, value=0.6228}
         {arriveTangent=0.0224, interpMode=RCIM_Cubic, leaveTangent=0.0224, tangentMode=RCTM_User, time=1, value=1}
        ]}}}))
   &ScaleColor(
    Linear Color Curve=DataInterface() {propertyValues {
      AlphaCurve {keys [<at 0.0079 0>,<at 0.1937 1>,<at 0.9371 0>]}
      BlueCurve {keys [<at 0.0297 0.0208>]}
      GreenCurve {keys [<at 0.0297 0.0069>]}
      RedCurve {keys [<at 0.0297 0.0174>]}
     }}
    Scale Mode=::NewEnumerator2
   )
   &ScaleColor(
    Scale Alpha=&FloatFromCurve(
     FloatCurve=DataInterface() {propertyValues {Curve {keys [
         {interpMode=RCIM_Cubic, tangentMode=RCTM_User, time=0, value=0}
         {interpMode=RCIM_Cubic, tangentMode=RCTM_User, time=0.5, value=1}
         {tangentMode=RCTM_User, time=1, value=0}
        ]}}}
     CurveIndex=$Particles.RibbonLinkOrder
     Scale Curve=15
    )
    ScaleA=true
    ScaleRGB=true
   ) {module=ScaleColor001}
  ]
  rendererClasses [&NiagaraRibbonRendererProperties]
  renderers [{rendererClass=&NiagaraRibbonRendererProperties}]
 }}
summary {
 systemName=NS_AI_Bosses_Puppeteer_WarpTrail_PROTO
 emitters [{emitterName=Trail002, rendererClasses [&NiagaraRibbonRendererProperties]}]
 userVariables [
  User.IsRoseDepthTeleport {type=NiagaraBool}
  User.TargetPosition {type=NiagaraPosition, description="", defaultValue {}}
  User.TargetPositionSpace {
   type=ENiagaraCoordinateSpace
   description=""
   defaultValue {
    struct=NiagaraExt_VariableValue_Enum
    value {displayName=Simulation, enum=/Script/Niagara.ENiagaraCoordinateSpace, enumName=ENiagaraCoordinateSpace::Simulation}
   }
  }
 ]
}
